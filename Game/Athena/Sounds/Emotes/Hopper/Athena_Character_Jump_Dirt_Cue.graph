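# Sound cue for a character jump on dirt, stored with the Hopper emote sounds.
# A Random node picks one of three equally weighted Commando jump foley waves, then a Modulator varies its volume and raises its pitch by a random amount.
# The cue uses emote music class, attenuation and concurrency assets rather than foley ones, and has a send to the music audio analysis submix.
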SoundCue /Game/Athena/Sounds/Emotes/Hopper/Athena_Character_Jump_Dirt_Cue
root=Modulator_2
nodes [
 {id=Modulator_2, children [Random_0], properties {pitchMax=1.3, pitchMin=1.1, volumeMax=1, volumeMin=0.7}}
 {id=Random_0, children [WavePlayer_349, WavePlayer_352, WavePlayer_353], properties {bShouldExcludeFromBranchCulling=true, weights [1,1,1]}}
 {id=WavePlayer_349, properties {soundWaveAssetPtr=/Game/Sounds/Fortress_Foley/Commando/Commando_Jump_006}}
 {id=WavePlayer_352, properties {soundWaveAssetPtr=/Game/Sounds/Fortress_Foley/Commando/Commando_Jump_003}}
 {id=WavePlayer_353, properties {soundWaveAssetPtr=/Game/Sounds/Fortress_Foley/Commando/Commando_Jump_004}}
]
properties {
 attenuationSettings=&Fort_Emotes_Athena_Att
 maxDistance=1600
 soundClassObject=&Fort_Music_Emote
 concurrencyOverrides {bLimitToOwner=true, maxCount=1, resolutionRule=PreventNew}
 concurrencySet [/Game/Sounds/Concurrency_Rules/Athena_EmoteMusic_Conc]
 soundSubmixSends [{minSendLevel=0, soundSubmix=&Fortnite_AudioAnalysis_Music}]
}
